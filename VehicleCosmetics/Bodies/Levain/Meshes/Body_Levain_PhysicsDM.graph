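# Physics asset for the Levain vehicle body cosmetic, giving its mesh a simplified collision shape.
# A single box on the root bone, raised above the origin and shifted slightly along X, stands in for the body volume.
# It has no other bodies and no constraints, so it defines one rigid shape with no articulated parts.
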PhysicsAsset /VehicleCosmetics/Bodies/Levain/Meshes/Body_Levain_PhysicsDM
bodies {root [{shapeType=Box, extentX=444, extentY=216, extentZ=116, shapeName=root_box, center=<7.2828 0 102.2869>}]}
constraints []
